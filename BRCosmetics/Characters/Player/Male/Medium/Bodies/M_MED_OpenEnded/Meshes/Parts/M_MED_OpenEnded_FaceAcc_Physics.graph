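# Physics asset for the face accessory mesh of the M_MED_OpenEnded male medium player body, covering a dynamic bang bone chain on the head.
# Four capsule bodies run in series from a small sphere master body, which attaches to a large head sphere.
# Links along the chain have tight swing and twist limits, while the master-to-head joint leaves swing and twist free.
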PhysicsAsset /BRCosmetics/Characters/Player/Male/Medium/Bodies/M_MED_OpenEnded/Meshes/Parts/M_MED_OpenEnded_FaceAcc_Physics
bodies {
 dyn_bang_01 [{shapeType=Capsule, length=2, radius=1, shapeName=dyn_bang_01_capsule, center=<1 0 0>, rotation=<rot 90 0 0>}]
 dyn_bang_02 [{shapeType=Capsule, length=2, radius=1, shapeName=dyn_bang_02_capsule, center=<1 0 0>, rotation=<rot 90 0 0>}]
 dyn_bang_03 [{shapeType=Capsule, length=2, radius=1, shapeName=dyn_bang_03_capsule, center=<1 0 0>, rotation=<rot 90 0 0>}]
 dyn_bang_04 [{shapeType=Capsule, length=2, radius=1, shapeName=dyn_bang_04_capsule, center=<1 0 0>, rotation=<rot 90 0 0>}]
 dyn_bang_MSTR [{shapeType=Sphere, radius=0.505, shapeName=dyn_bang_MSTR_sphere}]
 head [{shapeType=Sphere, radius=9, shapeName=head_sphere, center=<9.398 1.6714 0>}]
}
constraints [
 {bone1Name=dyn_bang_01, bone2Name=dyn_bang_MSTR, swing1LimitDegrees=7, swing2LimitDegrees=4, twistLimitDegrees=1}
 {bone1Name=dyn_bang_02, bone2Name=dyn_bang_01, swing1LimitDegrees=7, swing2LimitDegrees=4, twistLimitDegrees=1}
 {bone1Name=dyn_bang_03, bone2Name=dyn_bang_02, swing1LimitDegrees=7, swing2LimitDegrees=4, twistLimitDegrees=1}
 {bone1Name=dyn_bang_04, bone2Name=dyn_bang_03, swing1LimitDegrees=7, swing2LimitDegrees=4, twistLimitDegrees=1}
 {
  bone1Name=dyn_bang_MSTR
  bone2Name=head
  swing1LimitDegrees=45
  swing1Motion=Free
  swing2LimitDegrees=45
  swing2Motion=Free
  twistLimitDegrees=45
  twistMotion=Free
 }
]
